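# Conversation database for the TikiTorch NPC, voiced in text as Zuko, giving his greeting dialogue.
# The entry point feeds one speech node whose contextual lines cover a first-ever meeting, reactions to three specific player outfits, and four general lines gated by match conversation count.
# Curve rows set priority for the first-meeting and outfit lines, with a default message as fallback.
# The speech then links out to the shared NPC quest and basic services conversations.
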ConversationDatabase /NPCLibrary/NPCs/TikiTorch/NPCLibrary_Conversation_TikiTorch
entry_points [{tag=Conversation.Entry.TikiTorch, destinations [#1]}]
nodes [
 ConversationEntryPointNode_7 {
  id=#1
  class=ConversationEntryPointNode
  outputs [#2]
  properties {entryTag {tagName=Conversation.Entry.TikiTorch}}
 }
 Speech_C_4 {
  id=#2
  class=Speech_C
  outputs [#3, #4]
  properties {
   nodeName=""
   generalConfig {
    defaultMessage=I must capture the Avatar.
    contextualMessages [
     {
      message=Hey. Zuko here.
      contextRequirements [{requirement {class=FortControllerRequirement_IsFirstEverConversationWithNPC}}]
      priority {curve {rowName=Tandem.Speech.Priority.FirstEverInteraction}}
     }
     {
      message=I understand what it's like to have the Fire Nation take something from you.
      contextRequirements [{requirement {class=FortControllerRequirement_HasCID, properties {allowedCIds [/BRCosmetics/Athena/Items/Cosmetics/Characters/Character_DirectContact]}}}]
      priority {curve {rowName=Tandem.Speech.Priority.OutfitReaction}}
     }
     {
      message="Hey, I'm, uh, sorry about burning your feet."
      contextRequirements [{requirement {class=FortControllerRequirement_HasCID, properties {allowedCIds [/BRCosmetics/Athena/Items/Cosmetics/Characters/Character_RemoteControl]}}}]
      priority {curve {rowName=Tandem.Speech.Priority.OutfitReaction}}
     }
     {
      message="I've spent years training for this day, and you're just a child!"
      contextRequirements [{requirement {class=FortControllerRequirement_HasCID, properties {allowedCIds [/BRCosmetics/Athena/Items/Cosmetics/Characters/Character_DoubleDuty]}}}]
      priority {curve {rowName=Tandem.Speech.Priority.OutfitReaction}}
     }
     {
      message=I must capture the Avatar.
      contextRequirements [{requirement {class=FortControllerRequirement_CheckNPCMatchConversationCount, properties {nPCUniqueIdTag {tagName=AISpawnerData.Type.Tandem.TikiTorch}}}}]
      priority {curve {rowName=None}}
     }
     {
      message=I wish Uncle Iroh was here. He'd know what to do.
      contextRequirements [{requirement {class=FortControllerRequirement_CheckNPCMatchConversationCount, properties {nPCUniqueIdTag {tagName=AISpawnerData.Type.Tandem.TikiTorch}}}}]
      priority {curve {rowName=None}}
     }
     {
      message=I will regain my honor. One day.
      contextRequirements [{requirement {class=FortControllerRequirement_CheckNPCMatchConversationCount, properties {nPCUniqueIdTag {tagName=AISpawnerData.Type.Tandem.TikiTorch}}}}]
      priority {curve {rowName=None}}
     }
     {
      message=Finding the Avatar is my destiny.
      contextRequirements [{requirement {class=FortControllerRequirement_CheckNPCMatchConversationCount, properties {nPCUniqueIdTag {tagName=AISpawnerData.Type.Tandem.TikiTorch}}}}]
      priority {curve {rowName=None}}
     }
    ]
   }
  }
 }
 ConversationLinkNode_0 {id=#4, class=ConversationLinkNode, properties {remoteEntryTag {tagName=Conversation.Entry.BasicServices}}}
 ConversationLinkNode_1 {id=#3, class=ConversationLinkNode, properties {remoteEntryTag {tagName=Conversation.Entry.NPCLibrary.Quests}}}
]
